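# Niagara system for the Season 20 quest detector back bling, drawing blue and green indicator lights on the prop's skeletal mesh.
# BlueLight bursts short flashes with a sound, looping at one over User.BlinkSpeed and gated by comparing that speed to zero.
# GreenLight holds a persistent particle whose alpha follows User.bGreenLightEnabled.
# Both lights use sprite and mesh renderers with camera offset; the FN_Debug emitter is disabled.
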NiagaraSystem /BRCosmetics/Animation/Props/Gadgets_Backbling_Detector/FX/NS_Quest_S20_Narrative_W8_Q4_BackSensor
dependencies {
 usedDataInterfaces [&NiagaraDataInterfaceAudioPlayer]
 usedDynamicInputs [
  &ClampFloat
  &Subtract_Float
  &VectorLength
  &Subtract_Vector
  /Niagara/DynamicInputs/Helpers/SimulationPosition
  &DotProduct
  &NormalizeVector
  &Multiply_Vector
  &ConvertPositionToVector
 ]
 usedModules [
  &SystemState
  &self:SystemSpawnScript.SystemScriptSource.SystemScriptGraph.NiagaraNodeAssignment_0.SetVariables_#1
  &self:SystemSpawnScript.SystemScriptSource.SystemScriptGraph.NiagaraNodeAssignment_1.SetVariables_#2
 ]
}
emitters {
 BlueLight {
  emitterSpawnScript [
   &SampleSkeletalMesh(
    AdditionalYawPitchRoll=false
    Bone / Socket Index=1
    Bone Sampling Mode:ENiagaraBoneSamplingMode=::NewEnumerator4
    Interpolate Mesh Sampling=false
    Position Offset Coordinate Space:ENiagaraCoordinateSpace=::Local
    Sampled Position Coordinate Space:ENiagaraCoordinateSpace=::Local
    Skeletal Mesh:NiagaraDataInterfaceSkeletalMesh=$User.SkeletalMesh
   )
   &InitializeAudioPlayer()
  ]
  emitterUpdateScript [
   &EmitterState(
    Inactive Response=::NewEnumerator1
    Life Cycle Mode=::NewEnumerator1
    Loop Duration=&Divide_Float(A=1, B=$User.BlinkSpeed)
    Recalculate Duration Each Loop=true
    UseLoopDelay=false
   )
   &SpawnBurst_Instantaneous(Spawn Group=1, Use Spawn Probability=false, UseLoopCountLimit=false) {module=SpawnBurst_Instantaneous001}
   /SetVariables(Emitter.IsEnabled:NiagaraBool=&SetBoolByFloatComparison(A=$User.BlinkSpeed, B=0, Comparison Type=::NewEnumerator5)) {isSetParameters=true}
  ]
  particleSpawnScript [
   &InitializeParticle(
    Color=<rgba 0 0.2 1 1>
    Lifetime=0.133
    Position Mode=::NewEnumerator0
    Sprite Size Mode=::NewEnumerator3
    Uniform Sprite Size=9
    Write Material Random=false
   )
   /Niagara/Modules/Audio/PlayPersistentAudio(Audio Player=DataInterface() {propertyValues {
      Attenuation=null
      Concurrency=null
      MaxPlaysPerTick=10
      SoundToPlay=/BRCosmetics/Sounds/Other/Quests/InvestigateTheReadings/Spot_Backbling_Detector_Light_Flash_Cue
      bAllowLoopingOneShotSounds=false
      bLimitPlaysPerTick=true
      bOnlyActiveDuringGameplay=false
      bStopWhenComponentIsDestroyed=true
      ConfigurationUserParameter {parameter {name=None, typeDefHandle {registeredTypeIndex=1}, varData []}}
      ParameterNames []
     }})
  ]
  particleUpdateScript [
   &ParticleState()
   &ScaleColor(ScaleA=false, ScaleRGB=false)
   &CameraOffset(Camera Offset Amount=9)
   /Niagara/Modules/Audio/UpdatePersistentAudio(Update Type:ENiagara_AudioParamType=::NewEnumerator5)
  ]
  rendererClasses [&NiagaraSpriteRendererProperties, &NiagaraMeshRendererProperties]
  renderers [
   {rendererClass=&NiagaraSpriteRendererProperties}
   {rendererClass=&NiagaraMeshRendererProperties, rendererIndex=1}
  ]
 }
 FN_Debug {
  bEnabled=false
  emitterUpdateScript [
   &EmitterState()
   &SpawnBurst_Instantaneous(Spawn Count=1, Use Spawn Probability=false)
  ]
  particleSpawnScript [/Niagara/Modules/Spawn/Initialization/InitializeParticle(
    Lifetime=5
    Module.Write Color=true
    Module.Write Lifetime=true
    Module.Write Mass=true
    Module.Write Material Random=true
    Module.Write Position=false
    Module.Write Scale=false
    Module.Write SpriteRotation=false
    Module.Write SpriteSize=true
    Position=/Niagara/Functions/Localspace/SimulationPosition()
    Sprite Size=<44 44>
   )]
  particleUpdateScript [
   &ParticleState(Kill Particles When Lifetime Has Elapsed=false)
   &FN_Timer()
   /SetVariables(Particles.Debug=0.1) {isSetParameters=true}
   &DynamicMaterialParameters(
    Index 0 Param 1=$Particles.Debug
    Module.Param0WriteEnabled=true
    Module.Param1WriteEnabled=true
    Module.Param2WriteEnabled=true
    Module.Param3WriteEnabled=true
   )
   &CameraOffset(Camera Offset Amount=33)
  ]
  rendererClasses [&NiagaraSpriteRendererProperties]
  renderers [{rendererClass=&NiagaraSpriteRendererProperties}]
 }
 GreenLight {
  emitterSpawnScript [&SampleSkeletalMesh(
    AdditionalYawPitchRoll=false
    Bone Sampling Mode:ENiagaraBoneSamplingMode=::NewEnumerator4
    Interpolate Mesh Sampling=false
    Position Offset Coordinate Space:ENiagaraCoordinateSpace=::Local
    Sampled Position Coordinate Space:ENiagaraCoordinateSpace=::Local
    Skeletal Mesh:NiagaraDataInterfaceSkeletalMesh=$User.SkeletalMesh
   )]
  emitterUpdateScript [
   &EmitterState(
    Inactive Response=::NewEnumerator1
    Life Cycle Mode=::NewEnumerator1
    Loop Behavior=::NewEnumerator1
    Loop Duration Mode=::NewEnumerator1
    UseLoopDelay=false
   )
   &SpawnBurst_Instantaneous(Use Spawn Probability=false, UseLoopCountLimit=false)
  ]
  particleSpawnScript [/Niagara/Modules/Spawn/Initialization/V2/InitializeParticle(
    Color=<rgba 0.1 1 0 0.5>
    Lifetime=5
    Position Mode=::NewEnumerator0
    Sprite Size Mode=::NewEnumerator3
    Uniform Sprite Size=9
    Write Material Random=false
   )]
  particleUpdateScript [
   &ParticleState(Kill Particles When Lifetime Has Elapsed=false)
   &ScaleColor(
    Scale Alpha=&MakeCustomFloatFromBool(Bool=$User.bGreenLightEnabled, True Float=1, False Float=0)
    ScaleA=true
    ScaleRGB=false
   )
   &CameraOffset(Camera Offset Amount=9)
  ]
  rendererClasses [&NiagaraSpriteRendererProperties, &NiagaraMeshRendererProperties]
  renderers [
   {rendererClass=&NiagaraSpriteRendererProperties}
   {rendererClass=&NiagaraMeshRendererProperties, rendererIndex=1}
  ]
 }
}
summary {
 systemName=NS_Quest_S20_Narrative_W8_Q4_BackSensor
 emitters [
  {bEnabled=false, emitterName=FN_Debug, rendererClasses [&NiagaraSpriteRendererProperties]}
  {emitterName=GreenLight, rendererClasses [&NiagaraSpriteRendererProperties, &NiagaraMeshRendererProperties]}
  {emitterName=BlueLight, rendererClasses [&NiagaraSpriteRendererProperties, &NiagaraMeshRendererProperties]}
 ]
 userVariables [
  User.bGreenLightEnabled {type=NiagaraBool, defaultValue=true}
  User.BlinkSpeed {type=NiagaraFloat, defaultValue=0, description=""}
  User.ClosestObjectPosition {type=NiagaraPosition, description="", defaultValue {}}
  User.GreenLightToggleDistance {type=NiagaraFloat, defaultValue=350, description=""}
  User.SkeletalMesh {
   type=NiagaraDataInterfaceSkeletalMesh
   defaultValue {
    struct=NiagaraExt_VariableValue_DataInterface
    value {dataInterface=&self:NiagaraDataInterfaceSkeletalMesh_0}
   }
  }
 ]
}
