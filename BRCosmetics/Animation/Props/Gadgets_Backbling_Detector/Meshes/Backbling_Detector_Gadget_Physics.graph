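# Physics asset for the detector gadget back bling cosmetic, giving simple collision bodies and joints to its bag, tube and antenna.
# A five-capsule antenna chain tapers from the bag to the tip, with small swing limit angles that widen along the chain.
# The base antenna joint locks one swing axis.
# The tube is a sphere jointed directly to the bag capsule, with much wider swing and twist limit angles.
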PhysicsAsset /BRCosmetics/Animation/Props/Gadgets_Backbling_Detector/Meshes/Backbling_Detector_Gadget_Physics
bodies {
 antenna_01 [{shapeType=Capsule, length=4, radius=1, shapeName=antenna_01_capsule, center=<2.2 0 0>, rotation=<rot 90 0 0>}]
 antenna_02 [{shapeType=Capsule, length=5.5, radius=0.6, shapeName=antenna_02_capsule, center=<3.3 0 0>, rotation=<rot 90 0 0>}]
 antenna_03 [{shapeType=Capsule, length=5, radius=0.4, shapeName=antenna_03_capsule, center=<2.6 0 0>, rotation=<rot 90 0 0>}]
 antenna_04 [{shapeType=Capsule, length=6.5, radius=0.25, shapeName=antenna_04_capsule, center=<3.3 0 0>, rotation=<rot 90 0 0>}]
 antenna_05 [{shapeType=Capsule, length=5.5, radius=0.25, shapeName=antenna_05_capsule, center=<3 0 0>, rotation=<rot 90 0 0>}]
 bag [{shapeType=Capsule, length=5, radius=5, shapeName=bag_capsule}]
 tube [{shapeType=Sphere, radius=3, shapeName=tube_sphere}]
}
constraints [
 {bone1Name=antenna_01, bone2Name=bag, swing1LimitDegrees=2, swing1Motion=Locked, swing2LimitDegrees=2, twistLimitDegrees=1}
 {bone1Name=antenna_02, bone2Name=antenna_01, swing1LimitDegrees=2, swing2LimitDegrees=2, twistLimitDegrees=1}
 {bone1Name=antenna_03, bone2Name=antenna_02, swing1LimitDegrees=3, swing2LimitDegrees=3, twistLimitDegrees=1}
 {bone1Name=antenna_04, bone2Name=antenna_03, swing1LimitDegrees=4, swing2LimitDegrees=4, twistLimitDegrees=1}
 {bone1Name=antenna_05, bone2Name=antenna_04, swing1LimitDegrees=5, swing2LimitDegrees=5, twistLimitDegrees=1}
 {bone1Name=tube, bone2Name=bag, swing1LimitDegrees=90, swing2LimitDegrees=90, twistLimitDegrees=90}
]
